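# Sound cue for short wood-stump slide foley of the Mimosa biped cosmetic companion.
# A quality-level node feeds the first two quality slots a random pick of seven wave variations and the last two a smaller set of four.
# Both random sets are equally weighted and share the first four waves.
# It plays at reduced volume in the pets sound class, using companion foley attenuation and the biped footstep concurrency.
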SoundCue /CosmeticCompanions/Assets/Audio_Shared/Foley/Biped/WoodStump/SC_MImosa_Biped_WoodStump_SlideShort
root=QualityLevel_0
nodes [
 {id=QualityLevel_0, children [Random_3,Random_3,Random_5,Random_5]}
 {
  id=Random_3
  children [WavePlayer_11, WavePlayer_12, WavePlayer_13, WavePlayer_14, WavePlayer_15, WavePlayer_16, WavePlayer_19]
  properties {weights [1,1,1,1,1,1,1]}
 }
 {id=Random_5, children [WavePlayer_20, WavePlayer_21, WavePlayer_22, WavePlayer_23], properties {weights [1,1,1,1]}}
 {
  id=WavePlayer_11
  properties {soundWaveAssetPtr=/CosmeticCompanions/Assets/Audio_Shared/Foley/Biped/WoodStump/Wavs/Mimosa_Biped_WoodStump_SlideShort_07}
 }
 {
  id=WavePlayer_12
  properties {soundWaveAssetPtr=/CosmeticCompanions/Assets/Audio_Shared/Foley/Biped/WoodStump/Wavs/Mimosa_Biped_WoodStump_SlideShort_06}
 }
 {
  id=WavePlayer_13
  properties {soundWaveAssetPtr=/CosmeticCompanions/Assets/Audio_Shared/Foley/Biped/WoodStump/Wavs/Mimosa_Biped_WoodStump_SlideShort_05}
 }
 {
  id=WavePlayer_14
  properties {soundWaveAssetPtr=/CosmeticCompanions/Assets/Audio_Shared/Foley/Biped/WoodStump/Wavs/Mimosa_Biped_WoodStump_SlideShort_04}
 }
 {
  id=WavePlayer_15
  properties {soundWaveAssetPtr=/CosmeticCompanions/Assets/Audio_Shared/Foley/Biped/WoodStump/Wavs/Mimosa_Biped_WoodStump_SlideShort_03}
 }
 {
  id=WavePlayer_16
  properties {soundWaveAssetPtr=/CosmeticCompanions/Assets/Audio_Shared/Foley/Biped/WoodStump/Wavs/Mimosa_Biped_WoodStump_SlideShort_02}
 }
 {
  id=WavePlayer_19
  properties {soundWaveAssetPtr=/CosmeticCompanions/Assets/Audio_Shared/Foley/Biped/WoodStump/Wavs/Mimosa_Biped_WoodStump_SlideShort_01}
 }
 {
  id=WavePlayer_20
  properties {soundWaveAssetPtr=&Mimosa_Biped_WoodStump_SlideShort_04}
 }
 {
  id=WavePlayer_21
  properties {soundWaveAssetPtr=&Mimosa_Biped_WoodStump_SlideShort_03}
 }
 {
  id=WavePlayer_22
  properties {soundWaveAssetPtr=&Mimosa_Biped_WoodStump_SlideShort_02}
 }
 {
  id=WavePlayer_23
  properties {soundWaveAssetPtr=&Mimosa_Biped_WoodStump_SlideShort_01}
 }
]
properties {
 attenuationSettings=/Game/Sounds/Attenuation/AS_Companion_Foley
 maxDistance=1200
 priority=0
 soundClassObject=&Fort_Creatures_Pets
 volumeMultiplier=0.5
 concurrencySet [/CosmeticCompanions/Assets/Audio_Shared/Foley/Biped/CONC_Mimosa_Biped_Footstep]
}
